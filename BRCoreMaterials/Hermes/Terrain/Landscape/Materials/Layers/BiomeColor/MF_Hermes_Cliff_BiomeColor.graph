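# Landscape material function that tints cliff base color per terrain biome for the Hermes terrain.
# It unpacks packed terrain mask inputs into biome masks and partially desaturates the incoming BaseColor.
# Chained lerps layer saturated multiply tints for RedForest, Mountains, MaskMaker and Bigsplat, and a static switch controls whether biome color is written back to the material attributes.
# The moss tint blend and QuadD input are left unconnected.
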MaterialFunction /BRCoreMaterials/Hermes/Terrain/Landscape/Materials/Layers/BiomeColor/MF_Hermes_Cliff_BiomeColor
graph {
 properties {libraryCategoriesText [FN-Landscape]}
 nodes [
  FunctionInput() {bUsePreviewValueAsDefault=true, inputName=TerrainMasks_QuadD, inputType=FunctionInput_Vector4, sortPriority=40}
  n1 = NamedRerouteUsage()
  n2 = NamedRerouteUsage()
  FunctionOutput(None=SetMaterialAttributes!(
    MaterialAttributes=NamedRerouteUsage()
    BaseColor=StaticSwitchParameter(True=@n1, False=@n2) {defaultValue=true, parameterName=UseHermesBIomeColors}
   ) {attributeSetTypes [#1]})
  n3 = Saturate(None=Multiply(A=VectorParameter() {group=BiomeColors, parameterName=CliffColor_Mountains, defaultValue=<rgba 0.35 0.35 0.35 1>}.RGB, B=NamedRerouteUsage()))
  LinearInterpolate(
   A=@n3
   B=Multiply(B=VectorParameter() {group=BiomeColors, parameterName=CliffColor_Something_Moss, defaultValue=<rgba 0.4067 0.4792 0.35 0>}.RGB)
   Alpha=NamedRerouteUsage()
  )
  NamedRerouteDeclaration(None=FunctionInput() {bUsePreviewValueAsDefault=true, inputName=MossMask, inputType=FunctionInput_Scalar, sortPriority=100}) {name=MossMask, nodeColor=<rgba 0.5 0 1 1>}
  n4 = MaterialFunctionCall(Float4=FunctionInput() {bUsePreviewValueAsDefault=true, inputName=TerrainMasks_QuadB, inputType=FunctionInput_Vector4, sortPriority=20}) {materialFunction=&BreakOutFloat4Components}
  NamedRerouteDeclaration(None=@n4.R) {name=BM_Goldlands, nodeColor=<rgba 0.5 0 1 1>}
  n5 = NamedRerouteDeclaration(None=NamedRerouteUsage()) {name=NoBiomeColor, nodeColor=<rgba 0.5 0 1 1>}
  n6 = Saturate(None=Multiply(A=VectorParameter() {group=BiomeColors, parameterName=CliffColor_RedForest, defaultValue=<rgba 1.5 1.5 1.5 1>}.RGB, B=NamedRerouteUsage()))
  n7 = Saturate(None=Subtract(A=NamedRerouteUsage(), B=NamedRerouteUsage()))
  n8 = Saturate(None=Subtract(A=NamedRerouteUsage(), B=NamedRerouteUsage()))
  n9 = NamedRerouteUsage()
  n10 = VectorParameter() {group=BiomeColors, parameterName=CliffColor_MaskMaker, defaultValue=<rgba 0.24 0.2623 0.3 1>}
  n11 = NamedRerouteUsage()
  n12 = LinearInterpolate(
   A=LinearInterpolate(A=LinearInterpolate(A=@n5, B=@n6, Alpha=@n7), B=@n3, Alpha=Max(A=@n8, B=@n9))
   B=Saturate(None=Multiply(A=@n10.RGB, B=@n11))
   Alpha=NamedRerouteUsage()
  )
  n13 = Saturate(None=Multiply(A=VectorParameter() {group=BiomeColors, parameterName=CliffColor_Bigsplat, defaultValue=<rgba 1 1 1 1>}.RGB, B=NamedRerouteUsage()))
  n14 = NamedRerouteUsage()
  NamedRerouteDeclaration(None=ComponentMask(None=LinearInterpolate(A=@n12, B=@n13, Alpha=@n14)) {b=true, g=true, r=true}) {name=OutputBasecolor, nodeColor=<rgba 0.5 0 1 1>}
  n15 = GetMaterialAttributes(None=FunctionInput(Preview=SetMaterialAttributes!(MaterialAttributes)) {bUsePreviewValueAsDefault=true, inputName=Material, inputType=FunctionInput_MaterialAttributes}) {attributeGetTypes [#1]}
  NamedRerouteDeclaration(None=Desaturation(
    None=NamedRerouteDeclaration(None=@n15.BaseColor) {name=Input_BaseColor, nodeColor=<rgba 0.5 0 1 1>}
    Fraction=ScalarParameter() {defaultValue=0.5, group=BiomeColors, parameterName=Cliff_BiomeDesat}
   )) {name=BaseColorDesat, nodeColor=<rgba 0.5 0 1 1>}
  NamedRerouteDeclaration(None=@n15.MaterialAttributes) {name=InputMaterial, nodeColor=<rgba 0.5 0 1 1>}
  NamedRerouteDeclaration(None=@n4.G) {name=BM_RedForest, nodeColor=<rgba 0.5 0 1 1>}
  n16 = MaterialFunctionCall(Float4=FunctionInput() {bUsePreviewValueAsDefault=true, inputName=TerrainMasks_QuadA, inputType=FunctionInput_Vector4, sortPriority=10}) {materialFunction=&BreakOutFloat4Components}
  NamedRerouteDeclaration(None=@n16.R) {name=BM_Mountains, nodeColor=<rgba 0.5 0 1 1>}
  n17 = MaterialFunctionCall(Float4=FunctionInput() {bUsePreviewValueAsDefault=true, inputName=TerrainMasks_QuadC, inputType=FunctionInput_Vector4, sortPriority=30}) {materialFunction=&BreakOutFloat4Components}
  NamedRerouteDeclaration(None=@n17.G) {name=BM_MaskMaker, nodeColor=<rgba 0.5 0 1 1>}
  NamedRerouteDeclaration(None=@n17.B) {name=BM_Bigsplat, nodeColor=<rgba 0.5 0 1 1>}
  NamedRerouteDeclaration(None=@n16.G) {name=BM_Wetlands, nodeColor=<rgba 0.5 0 1 1>}
  NamedRerouteDeclaration(None=@n16.B) {name=BM_MossyForest, nodeColor=<rgba 0.5 0 1 1>}
  NamedRerouteDeclaration(None=@n16.A) {name=BM_BorealFall, nodeColor=<rgba 0.5 0 1 1>}
 ]
}
