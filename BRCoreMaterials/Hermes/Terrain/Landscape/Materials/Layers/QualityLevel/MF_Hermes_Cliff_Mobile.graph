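# Material function for the mobile cliff layer of the Hermes landscape, returning material attributes passed through the cliff biome color function with four terrain mask inputs.
# Samples one packed mask texture biplanarly in world XZ and YZ, flipping a channel by vertex normal sign and blending with a sharpness-driven mask.
# Derives a slope-based height mask, a three-color base color, specular and roughness from masks and tunable Mobile Cliff parameters, avoiding separate color textures.
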MaterialFunction /BRCoreMaterials/Hermes/Terrain/Landscape/Materials/Layers/QualityLevel/MF_Hermes_Cliff_Mobile
graph {nodes [
  n1 = SetMaterialAttributes!(
   MaterialAttributes
   BaseColor=NamedRerouteUsage()
   Specular=NamedRerouteUsage()
   Roughness=NamedRerouteUsage()
   Normal=ComponentMask(None=NamedRerouteUsage()) {b=true, g=true, r=true}
   OpacityMask=NamedRerouteUsage()
  ) {
   desc=HighEnd Material
   attributeSetTypes [
    #1
    #2
    #3
    #4
    #5
   ]
  }
  n2 = Constant()
  n3 = FunctionInput(Preview=Constant()) {bUsePreviewValueAsDefault=true, inputName=TerrainMasks_QuadB, inputType=FunctionInput_Vector4, sortPriority=20}
  n4 = Constant()
  n5 = Constant()
  FunctionOutput(None=MaterialFunctionCall!(
    Material=@n1
    TerrainMasks_QuadA=FunctionInput(Preview=@n2) {bUsePreviewValueAsDefault=true, inputName=TerrainMasks_QuadA, inputType=FunctionInput_Vector4, sortPriority=10}
    TerrainMasks_QuadB=NamedRerouteDeclaration(None=@n3) {name=Moss Desaturation Mask, nodeColor=<rgba 0.2422 0 1 1>}
    TerrainMasks_QuadC=FunctionInput(Preview=@n4) {bUsePreviewValueAsDefault=true, inputName=TerrainMasks_QuadC, inputType=FunctionInput_Vector4, sortPriority=30}
    TerrainMasks_QuadD=FunctionInput(Preview=@n5) {bUsePreviewValueAsDefault=true, inputName=TerrainMasks_QuadD, inputType=FunctionInput_Vector4, sortPriority=40}
    MossMask
   ) {materialFunction=/BRCoreMaterials/Hermes/Terrain/Landscape/Materials/Layers/BiomeColor/MF_Hermes_Cliff_BiomeColor}.Result)
  n6 = ScalarParameter() {defaultValue=3400, group=Mobile Cliff, parameterName=Mobile Cliff Scale X}
  n7 = Divide(
   A=TruncateLWC!(None=WorldPosition().XYZ)
   B=AppendVector(A=AppendVector(A=@n6, B=@n6), B=ScalarParameter() {defaultValue=2482, group=Mobile Cliff, parameterName=Mobile Cliff Scale Y})
  ) {constB=1200}
  n8 = TextureObjectParameter() {
   group=Mobile Cliff
   parameterName=Mobile Cliff Texture
   samplerType=SAMPLERTYPE_Masks
   texture=/Hermes/Terrain/Textures/Mobile/T_Hermes_Mobile_Cliff1
  }
  n9 = TextureSample(UVs=ComponentMask(None=@n7) {b=true, r=true}, Tex=@n8) {samplerType=SAMPLERTYPE_Masks, texture=/Game/Environments/Helios/Terrain/TexturesMobile/T_Helios_Mobile_Cliff}
  n10 = VertexNormalWS()
  n11 = Sign(None=ComponentMask(None=@n10) {g=true})
  n12 = Abs(None=Subtract(A=Saturate(None=@n11), B=@n9.G))
  n13 = Sign(None=ComponentMask(None=@n10) {r=true})
  n14 = TextureSample(UVs=ComponentMask(None=@n7) {b=true, g=true}, Tex=@n8) {samplerType=SAMPLERTYPE_Masks, texture=&T_Helios_Mobile_Cliff}
  n15 = Abs(None=Subtract(A=Saturate(None=@n13), B=@n14.G))
  NamedRerouteDeclaration(None=LinearInterpolate(
    A=MaterialFunctionCall(R=@n9.R, G=@n12, B=@n9.R, A=@n9.R) {materialFunction=&AppendMany}.RGBA
    B=MaterialFunctionCall(R=@n15, G=@n14.R, B=@n14.R, A=@n14.R) {materialFunction=&AppendMany}.RGBA
    Alpha=MaterialFunctionCall!(Sharpness) {materialFunction=/Game/Athena/Helios/Environments/Landscape/Materials/MaterialFunctions_Utilities/MF_MobileCliffBiplanarMask}.Result
   )) {name=Texture, nodeColor=<rgba 0 0.2031 1 1>}
  n16 = ComponentMask(None=VertexNormalWS()) {b=true}
  n17 = Multiply(A=OneMinus(None=Abs(None=@n16))) {constB=5}
  n18 = ComponentMask(None=NamedRerouteUsage()) {b=true}
  n19 = Saturate(None=Subtract(A=Multiply(A=@n18) {constB=4}) {constB=0.3})
  NamedRerouteDeclaration(None=Saturate(None=Subtract(A=@n17, B=@n19))) {name=Height, nodeColor=<rgba 0.8984 1 0 1>}
  n20 = LinearInterpolate(
   A=VectorParameter() {group=Mobile Cliff, parameterName=Mobile Cliff Color 1, defaultValue=<rgba 0.1458 0.0948 0.0592 1>}.RGB
   B=VectorParameter() {group=Mobile Cliff, parameterName=Mobile Cliff Color 2, defaultValue=<rgba 0.2331 0.25 0.2212 1>}.RGB
   Alpha=ComponentMask(None=NamedRerouteUsage()) {a=true}
  )
  n21 = VectorParameter() {group=Mobile Cliff, parameterName=Mobile Cliff Color 3, defaultValue=<rgba 0.6586 0.6776 0.7344 1>}
  n22 = ComponentMask(None=NamedRerouteUsage()) {b=true}
  n23 = ScalarParameter() {defaultValue=1.2, group=Mobile Cliff, parameterName=Mobile Cliff SecondColor Mult}
  n24 = Saturate(None=Subtract(
    A=Multiply(A=@n22, B=@n23) {constB=4}
    B=ScalarParameter() {defaultValue=0.55, group=Mobile Cliff, parameterName=Mobile Cliff SecondColor Sub}
   ) {constB=1.2})
  NamedRerouteDeclaration(None=Saturate(None=LinearInterpolate(A=@n20, B=@n21.RGB, Alpha=@n24))) {name=ClifflBaseColor, nodeColor=<rgba 0.5 0 1 1>}
  NamedRerouteDeclaration(None=FunctionInput() {bUsePreviewValueAsDefault=true, inputName=VariationMasks Mobile, inputType=FunctionInput_Vector4, sortPriority=3}) {name=VariationMasks Mobile, nodeColor=<rgba 1 0.7181 0.0414 1>}
  n25 = ComponentMask(None=NamedRerouteUsage()) {a=true}
  n26 = ScalarParameter() {defaultValue=0.6, group=Mobile Cliff, parameterName=Mobile Cliff Specular}
  NamedRerouteDeclaration(None=Saturate(None=Multiply(A=@n25, B=@n26))) {name=CliffSpecular, nodeColor=<rgba 0.5 0 1 1>}
  n27 = ComponentMask(None=NamedRerouteUsage()) {b=true}
  n28 = ScalarParameter() {defaultValue=2.5, group=Mobile Cliff, parameterName=Mobile Cliff Roughness}
  NamedRerouteDeclaration(None=Saturate(None=Multiply(A=@n27, B=@n28))) {name=CliffRoughness, nodeColor=<rgba 0.1484 0 1 1>}
 ]}
