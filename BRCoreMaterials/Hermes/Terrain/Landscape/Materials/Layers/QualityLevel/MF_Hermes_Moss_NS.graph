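# Moss landscape layer function for Hermes terrain that outputs base color, specular, roughness, normal and opacity mask as material attributes, using a mobile moss texture.
# One packed texture tiles in periodic, camera-relative world XY over 1024 units.
# The input's blue channel, doubled, blends that texture with its inverse for variation, yielding normal and height.
# Derived masks and tunable parameters drive a three-color tint with darkening, plus roughness.
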MaterialFunction /BRCoreMaterials/Hermes/Terrain/Landscape/Materials/Layers/QualityLevel/MF_Hermes_Moss_NS
graph {nodes [
  n1 = Constant() {r=1024}
  n2 = TransformPosition!(None=WorldPosition() {worldPositionShaderOffset=WPT_CameraRelative}.XYZ, Periodic World Tile Size=@n1) {transformSourceType=TRANSFORMPOSSOURCE_TranslatedWorld, transformType=TRANSFORMPOSSOURCE_PeriodicWorld}
  n3 = TextureSample(UVs=ComponentMask(None=Divide(A=@n2, B=@n1)) {g=true, r=true}) {samplerType=SAMPLERTYPE_Masks, texture=/Hermes/Terrain/Textures/Mobile/T_Hermes_Mobile_Moss}
  n4 = ComponentMask(None=FunctionInput() {inputType=FunctionInput_Vector4}) {b=true}
  n5 = Constant() {r=2}
  n6 = LinearInterpolate(
   A=@n3.RGBA
   B=OneMinus(None=@n3.RGBA)
   Alpha=NamedRerouteDeclaration(None=Multiply(A=@n4, B=@n5)) {name=Variation Mask, nodeColor=<rgba 1 0.7544 0.1303 1>}
  )
  ComponentMask(None=@n6) {b=true}
  n7 = SetMaterialAttributes!(
   MaterialAttributes
   BaseColor=NamedRerouteUsage()
   Specular=NamedRerouteUsage()
   Roughness=NamedRerouteUsage()
   Normal=NamedRerouteUsage()
   OpacityMask=NamedRerouteUsage()
  ) {
   desc=HighEnd Material
   attributeSetTypes [
    #1
    #2
    #3
    #4
    #5
   ]
  }
  FunctionOutput(None=@n7)
  NamedRerouteDeclaration(None=ComponentMask(None=@n6) {a=true}) {name=Height, nodeColor=<rgba 1 0 0.5156 1>}
  n8 = Multiply(A=NamedRerouteUsage(), B=ScalarParameter() {defaultValue=3, group=Mobile Moss, parameterName=Mobile Moss DetailMask Mult}) {constB=9}
  n9 = ScalarParameter() {defaultValue=1.5, group=Mobile Moss, parameterName=Mobile Moss DetailMask Sub}
  NamedRerouteDeclaration(None=Saturate(None=Subtract(A=@n8, B=@n9) {constB=2})) {name=Dirt Mask, nodeColor=<rgba 0.2656 1 0 1>}
  n10 = NamedRerouteUsage()
  NamedRerouteDeclaration(None=LinearInterpolate(
    A=ScalarParameter() {defaultValue=0.95, group=Mobile Moss, parameterName=Mobile Moss Roughness Min}
    B=ScalarParameter() {defaultValue=0.75, group=Mobile Moss, parameterName=Mobile Moss Roughness Max}
    Alpha=Multiply(A=@n10, B=@n10)
   )) {name=Roughness, nodeColor=<rgba 0 1 0.9531 1>}
  NamedRerouteDeclaration(None=ScalarParameter() {defaultValue=0.45, group=Mobile Moss, parameterName=Mobile Moss Specular}) {name=Specular, nodeColor=<rgba 0.0078 0 1 1>}
  NamedRerouteDeclaration(None=ComponentMask(None=@n6) {b=true, g=true, r=true}) {name=Normal, nodeColor=<rgba 0.9219 1 0 1>}
  n11 = LinearInterpolate(
   A=VectorParameter() {group=Mobile Moss, parameterName=Mobile Moss Color 1, defaultValue=<rgba 0.0278 0.0417 0.004 1>}.RGB
   B=VectorParameter() {group=Mobile Moss, parameterName=Mobile Moss Color 2, defaultValue=<rgba 0.0666 0.2708 0.0157 1>}.RGB
   Alpha=NamedRerouteUsage()
  )
  n12 = VectorParameter() {group=Mobile Moss, parameterName=Mobile Moss Color 3, defaultValue=<rgba 0.1307 0.276 0.0201 1>}
  n13 = NamedRerouteUsage()
  n14 = Abs(None=Add(A=NamedRerouteUsage()) {constB=-0.5})
  n15 = Multiply(A=Add(A=ComponentMask(None=@n14) {r=true}, B=ComponentMask(None=@n14) {r=true})) {constB=0.5}
  NamedRerouteDeclaration(None=Multiply(A=LinearInterpolate(A=@n11, B=@n12.RGB, Alpha=@n13), B=OneMinus(None=@n15))) {name=Base Color, nodeColor=<rgba 0.8984 1 0 1>}
  NamedRerouteDeclaration(None=@n3.B) {name=Texture Mask, nodeColor=<rgba 0.9688 0 1 1>}
 ]}
